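# Animation Blueprint for the DorsalDance large male head cosmetic, on its own custom head skeleton.
# On update, the event graph only calls the parent implementation.
# Unwired nodes hold sine-over-real-time oscillators: one remaps to a clamped Head_LOD1 morph target weight, the other sets GilAlpha.
# No execution path reaches those setters, and the file shows no anim graph reading GilAlpha.
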AnimBlueprint /BRCosmetics/Characters/Player/Male/Large/Heads/M_LRG_DorsalDance_Head/Meshes/M_LRG_DorsalDance_Head_AnimBP
target_skeleton=/BRCosmetics/Characters/Player/Male/Large/Heads/M_LRG_DorsalDance_Head/Meshes/CustomSkeleton/Fortnite_Base_Head_DorsalDance_Skeleton
graphs [EventGraph {nodes [
   n1 = Event BlueprintUpdateAnimation()
   > Parent: BlueprintUpdateAnimation!(execute, DeltaTimeX:float=@n1.DeltaTimeX) -> (then) {is=title K2Node_CallParentFunction}
   TryGetPawnOwner()
   n2 = "Sin (Radians)"(A=float * float(A=3, B=GetRealTimeSeconds()))
   SetMorphTarget!(
    execute
    self:Anim Instance Object Reference
    MorphTargetName:Name=Head_LOD1
    Value:float=Remap Float(Input=float / float(A=@n2, B=-1.4) {is=title K2Node_CallFunction}, OldRange=(X=-1,Y=1), NewRange=(X=0,Y=0.55), Clamp=true) -> (Output=~)
   ) -> (then)
   MakeVector(Y=0)
   n3 = float * float(A=GetRealTimeSeconds(), B=2.3)
   Set GilAlpha(GilAlpha:double=float / float(A="Sin (Radians)"(A=@n3), B=-1.4)) -> (Output_Get:double)
  ]}]
variables [GilAlpha {type=number}]
